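; Parent Blueprint for a lootable mini trophy case that trips security when searched.
; Construction fills mesh sockets with random decor, optionally seeded by world location, with special handling for gold bars.
; BeginPlay can override the loot tier group using playlist tags.
; On loot, the server alerts nearby AI once, triggers linked alarms, rumble, camera shake and a shatter cue, while remotes swap in broken glass and hide decor.
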(event EventBeginPlay
  (|CallParentFunction)
  (Class|FortAsyncActionCurrentPlaylistReady|PlaylistDataReadyAsync
    (:OnReady
      (if (Variables|Default|GetbScaleLootByPlaylist)
        (CallFunction|PickLootGroupbyPlaylist)))))

(event EventOnLoot
  (|CallParentFunction)
  (Networking|SwitchHasAuthority
    (:Authority
      (bind _returnvalue (Container|GetSearchingPawn))
      (CallFunction|AlertNearbyAIToPlayer 0 _returnvalue)
      (for _array_element (Variables|Default|GetNearbyAlarms)
        (Variables|/Hermes/Props/Gameplay/Gameplay_Alarm_A/BP_Gameplay_Alarm_A.BP_Gameplay_Alarm_A_C|SetAlertState true _array_element))
      (bind _returnvalue_2 (Container|GetSearchingPawn))
      (Game|Feedback|ClientPlayForceFeedback (Fortnite|GetFortPlayerControllerfromActor _returnvalue_2) "/Game/Weapons/ForceFeedback/FF_Generic_Explosion_Small_Near.FF_Generic_Explosion_Small_Near" "None" false false false) ; Rumble
      (Game|Feedback|ClientStartCameraShake :self (Fortnite|GetFortPlayerControllerfromActor _returnvalue_2) :Shake "/Game/Effects/Fort_Effects/BluePrints/CameraShakes/B_Small_Explosion_CameraShake.B_Small_Explosion_CameraShake_C" :Scale 1.0 :PlaySpace "CameraLocal") ; Camera Shake
      (GameplayCue|ExecuteGameplayCueOnActor(Burst) self (Variables|Default|GetCaseShatterCueTag) (Ability|GameplayCue|MakeGameplayCueParameters :NormalizedMagnitude 1.0 :RawMagnitude 1.0 :Location (Transformation|GetWorldLocation (Variables|Default|GetShatterLocation)) :Normal "0,0,0" :GameplayEffectLevel 1 :AbilityLevel 1 :TargetAttachComponent (Variables|Default|GetShatterLocation) :bReplicateLocationWhenUsingMinimalRepProxy false)) ; Glass shatter burst GCN
    )
    (:Remote
      (Components|StaticMesh|SetStaticMesh (Variables|Default|GetDisplayGlass) (Variables|Default|GetSM_BrokenGlass))
      (CallFunction|HideDecoMeshes))))

(fn AlertNearbyAIToPlayer
  (params (Instigator :type Controller%20Object%20Reference) (PlayerPawn :type Actor%20Object%20Reference))
  (bind _target (LocalVariables|GetTarget))
  (bind _instigator (Variables|Actor|GetInstigator))
  ; Make the AI Guards Around the Radio Tower get an AI Stimulus if someone interacts with the Radio Tower.
  (Networking|SwitchHasAuthority
    (:Authority
      (Utilities|FlowControl|DoOnce false
        (:Completed
          (bind (_outhits _returnvalue) (Collision|MultiSphereTraceForObjects (Transformation|GetActorLocation self) (+ (Transformation|GetActorLocation self) "0.000000,0.000000,100.000000") (ScalableFloat|GetValueatLevel (Variables|Default|GetRadiusToAlertAI)) (Variables|Default|GetPawnObjectTypes)))
          (if _returnvalue
            (for _array_element _outhits
              (bind _asfort_player_pawn_athena (Utilities|Casting|CastToFortPlayerPawnAthena (.HitActor (Collision|BreakHitResult _array_element))))
              (if (and (GameplayTags|HasMatchingGameplayTag _asfort_player_pawn_athena "(TagName=\"Pawn.Athena.NonParticipant\")") (not (GameplayTags|HasMatchingGameplayTag _asfort_player_pawn_athena "(TagName=\"AISpawnerData.Type.Tandem\")")))
                (bind _asfort_athena_aibot_controller (Utilities|Casting|CastToFortAthenaAIBotController (Pawn|GetController _asfort_player_pawn_athena)))
                (LocalVariables|SetAIBotController _asfort_athena_aibot_controller)
                (bind _asfort_player_pawn_athena_2 (Utilities|Casting|CastToFortPlayerPawnAthena (Variables|GetPlayerPawn))
                  (:then
                    (LocalVariables|SetTarget _asfort_player_pawn_athena_2))
                  (:CastFailed
                    (bind _asfort_player_controller (Utilities|Casting|CastToFortPlayerController _instigator)
                      (:then
                        (LocalVariables|SetTarget (Game|Player|GetPlayerPawn _asfort_player_controller)))
                      (:CastFailed
                        (bind _asfort_athena_aibot_controller_2 (Utilities|Casting|CastToFortAthenaAIBotController _instigator))
                        (if (not (GameplayTags|HasMatchingGameplayTag (Utilities|Casting|ToGameplayTagAssetInterface (Pawn|GetControlledPawn _asfort_athena_aibot_controller_2)) "(TagName=\"Pawn.Athena.NonParticipant\")"))
                          (LocalVariables|SetTarget (Pawn|GetControlledPawn _asfort_athena_aibot_controller_2)))))))
                (Utilities|IsValid _target
                  (:Is%20Valid
                    (Affiliation|ChangeAttitudeBetweenActors (Pawn|GetControlledPawn (LocalVariables|GetAIBotController)) _target "Hostile")
                    (Class|FortAthenaAIBotController|QueueStim (LocalVariables|GetAIBotController) _target (Transformation|GetActorLocation _target) (Transformation|GetActorForwardVector _target) "Enemy" 1e+01 "None"))
                  (:Is%20Not%20Valid
                    (Class|FortAthenaAIBotController|QueueStim (LocalVariables|GetAIBotController) self (Transformation|GetActorLocation self) (Transformation|GetActorForwardVector self) "Seen" 1e+01 "None")))))))))))

(fn HideDecoMeshes
  (for _array_element (Variables|Default|GetSMC_Deco)
    (Utilities|IsValid _array_element
      (:Is%20Valid
        (Rendering|SetVisibility _array_element false true)))))

(fn PickLootGroupByPlaylist
  (bind _keys (Utilities|Map|Keys (Variables|Default|GetLootGroupByPlaylistMap)))
  (bind (_array_element _array_index) (Utilities|Array|ForEachLoopwithBreak _keys)
    (:LoopBody
      (bind _returnvalue (Fortnite|GameplayTags|HasTagfromName (GameplayTags|GetTagName _array_element) (Class|FortGameStateAthena|GetAthenaPlaylistContextTags (Fortnite|Athena|GameState|GetGameStateAthena))))
      (if _returnvalue
        (bind _values (Utilities|Map|Values (Variables|Default|GetLootGroupByPlaylistMap)))
        (bind _output_get (Variables|Container|SetSearchLootTierGroup (Utilities|Array|Get(acopy) _values _array_index)))
        (Container|SetSearchLootTierGroup _output_get true true)))))

(fn SetRandomStream
  (bind _returnvalue (Math|Random|MakeRandomStream (Math|Float|Truncate (* (Class|EnvironmentArtFuncLib|RandomFloatbyLocation (Transformation|GetActorLocation) (Variables|Default|GetRandomizeDistance) true) 1e+04))))
  (Variables|Default|SetRandomStream _returnvalue))

(fn SpawnMesh
  (bind _staticmeshcomponent (Variables|Debug|GetStaticMeshComponent))
  (bind _returnvalue (Transformation|GetAllSocketNames _staticmeshcomponent))
  (if (Utilities|Array|IsNotEmpty (Variables|Default|GetAR_DecoMeshes))
    (if (> (Utilities|Array|Length _returnvalue) 0)
      (for _array_element _returnvalue
        (bind _returnvalue_2 (AddComponent|NODE_AddStaticMeshComponent-1 :RelativeTransform (Transformation|GetSocketTransform (Variables|Debug|GetStaticMeshComponent) _array_element "RTS_Actor")))
        (Utilities|Array|AddUnique (Variables|Default|GetSMC_Deco) _returnvalue_2)
        (Components|StaticMesh|SetStaticMesh _returnvalue_2 (Utilities|Array|RandomArrayItemfromStream (Variables|Default|GetAR_DecoMeshes) (Variables|Default|GetRandomStream)))
        (if (== (Variables|/Script/Engine.StaticMeshComponent|GetStaticMesh _returnvalue_2) "/Hermes/Props/Commerce/Commerce_GoldBars_A/Meshes/SM_Commerce_GoldBar_A.SM_Commerce_GoldBar_A")
          ; Adjusting scale and position of components that contain gold bars only
          (Transformation|SetRelativeTransform _returnvalue_2 (Math|Transform|MakeTransform (Math|Vector|MakeVector (+ (.x (.location (Transformation|GetRelativeTransform _returnvalue_2))) -5.0) (+ (.y (.location (Transformation|GetRelativeTransform _returnvalue_2))) -5.0) (.z (.location (Transformation|GetRelativeTransform _returnvalue_2)))) (.rotation (Transformation|GetRelativeTransform _returnvalue_2)) "0.750000,0.750000,0.750000") false false)
          ; Updating gold bars to Treasury material variant
          (Rendering|Material|SetMaterial _returnvalue_2 0 "/Hermes/Props/Treasury/Treasury_GoldBars_A/Materials/MI_Treasury_GoldBars_A.MI_Treasury_GoldBars_A"))))))

(fn ConstructionScript
  (|CallParentFunction)
  (if (Variables|Default|GetRandomizeByWorldPosition)
    (CallFunction|SetRandomStream)
    (CallFunction|SpawnMesh)))
