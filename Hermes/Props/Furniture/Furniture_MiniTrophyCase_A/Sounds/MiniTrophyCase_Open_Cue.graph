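# Sound cue for opening the mini trophy case prop in Hermes, layering an interact-end sound with a glass case smash.
# A mixer plays two modulated branches at once: one of three trophy case interact-end waves, and one of five borrowed jewelry case smash waves at slightly lower volume.
# It uses item-interact attenuation and overrides concurrency to allow one instance per owner, preventing new plays while one is active.
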SoundCue /Hermes/Props/Furniture/Furniture_MiniTrophyCase_A/Sounds/MiniTrophyCase_Open_Cue
root=Mixer_2
nodes [
 {id=Mixer_2, children [Modulator_2,Modulator_0], properties {inputVolume [1,0.8]}}
 {id=Modulator_0, children [Random_1]}
 {id=Modulator_2, children [Random_2]}
 {id=Random_1, children [WavePlayer_4,WavePlayer_5,WavePlayer_6,WavePlayer_7,WavePlayer_8], properties {weights [1,1,1,1,1]}}
 {id=Random_2, children [WavePlayer_9, WavePlayer_10, WavePlayer_11], properties {weights [1,1,1]}}
 {id=WavePlayer_10, properties {soundWaveAssetPtr=&/MiniTrophyCase_Interact_End_2}}
 {id=WavePlayer_11, properties {soundWaveAssetPtr=&/MiniTrophyCase_Interact_End_3}}
 {id=WavePlayer_4, properties {soundWaveAssetPtr=/Hermes/Audio/GlassCase/Hermes_JewelryCase_Smash_01}}
 {id=WavePlayer_5, properties {soundWaveAssetPtr=/Hermes/Audio/GlassCase/Hermes_JewelryCase_Smash_02}}
 {id=WavePlayer_6, properties {soundWaveAssetPtr=/Hermes/Audio/GlassCase/Hermes_JewelryCase_Smash_03}}
 {id=WavePlayer_7, properties {soundWaveAssetPtr=/Hermes/Audio/GlassCase/Hermes_JewelryCase_Smash_04}}
 {id=WavePlayer_8, properties {soundWaveAssetPtr=/Hermes/Audio/GlassCase/Hermes_JewelryCase_Smash_05}}
 {id=WavePlayer_9, properties {soundWaveAssetPtr=&/MiniTrophyCase_Interact_End_1}}
]
properties {
 attenuationSettings=/Game/Sounds/Attenuation/Fort_Item_Interact_Att
 bOverrideConcurrency=true
 maxDistance=1850
 soundClassObject=&Fort_Loot_Reward_PSM
 volumeMultiplier=1
 concurrencyOverrides {bLimitToOwner=true, maxCount=1, resolutionRule=PreventNew}
}
