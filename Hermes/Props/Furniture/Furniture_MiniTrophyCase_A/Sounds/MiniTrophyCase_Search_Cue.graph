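# Sound cue for searching the mini trophy case prop, using loot-collect sound class and item-interact attenuation, capped at four instances.
# A mixer layers two looping beds, one fixed wave at double input volume and one random pick of three, each pitch-modulated and enveloped to fade in.
# It also layers a random pitch-varied Start one-shot, with an empty slot in the pick, and a random End one-shot with no delay node.
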SoundCue /Hermes/Props/Furniture/Furniture_MiniTrophyCase_A/Sounds/MiniTrophyCase_Search_Cue
root=Mixer_1
nodes [
 {
  id=Enveloper_0
  children [Modulator_3]
  properties {pitchCurve {editorCurveData {keys [{time=0}]}}, volumeCurve {editorCurveData {keys [<at 0 0 RCIM_Cubic>,<at 1 1>]}}}
 }
 {
  id=Enveloper_1
  children [Modulator_1]
  properties {pitchCurve {editorCurveData {keys [{time=0}]}}, volumeCurve {editorCurveData {keys [<at 0 0 RCIM_Cubic>,<at 1 1>]}}}
 }
 {id=Mixer_1, children [Enveloper_0,Modulator_0,Random_0,Enveloper_1], properties {inputVolume [2,1,1,1]}}
 {id=Modulator_0, children [Random_2], properties {pitchMax=1.2, pitchMin=1, volumeMax=1, volumeMin=1}}
 {id=Modulator_1, children [Random_1], properties {pitchMax=1.1, pitchMin=0.7, volumeMax=1, volumeMin=1}}
 {id=Modulator_3, children [WavePlayer_10], properties {pitchMax=1.1, pitchMin=0.7, volumeMax=1, volumeMin=1}}
 {id=Random_0, children [WavePlayer_0,WavePlayer_1,WavePlayer_2], properties {bShouldExcludeFromBranchCulling=true, weights [1,1,1]}}
 {id=Random_1, children [WavePlayer_3,WavePlayer_4,WavePlayer_9], properties {bShouldExcludeFromBranchCulling=true, weights [1,1,1]}}
 {id=Random_2, children [WavePlayer_6,WavePlayer_7,WavePlayer_8,null], properties {bShouldExcludeFromBranchCulling=true, weights [1,1,1,1]}}
 {id=WavePlayer_0, properties {soundWaveAssetPtr=&/MiniTrophyCase_Interact_End_1}}
 {id=WavePlayer_1, properties {soundWaveAssetPtr=&/MiniTrophyCase_Interact_End_2}}
 {
  id=WavePlayer_10
  properties {bLooping=true, soundWaveAssetPtr=&/MiniTrophyCase_Interact_Loop_5}
 }
 {id=WavePlayer_2, properties {soundWaveAssetPtr=&/MiniTrophyCase_Interact_End_3}}
 {
  id=WavePlayer_3
  properties {bLooping=true, soundWaveAssetPtr=&/MiniTrophyCase_Interact_Loop_2}
 }
 {
  id=WavePlayer_4
  properties {bLooping=true, soundWaveAssetPtr=&/MiniTrophyCase_Interact_Loop_3}
 }
 {id=WavePlayer_6, properties {soundWaveAssetPtr=&/MiniTrophyCase_Interact_Start_1}}
 {id=WavePlayer_7, properties {soundWaveAssetPtr=&/MiniTrophyCase_Interact_Start_2}}
 {id=WavePlayer_8, properties {soundWaveAssetPtr=&/MiniTrophyCase_Interact_Start_3}}
 {
  id=WavePlayer_9
  properties {bLooping=true, soundWaveAssetPtr=&/MiniTrophyCase_Interact_Loop_4}
 }
]
properties {
 attenuationSettings=/Game/Sounds/Attenuation/Fort_Item_Interact_Att
 bOverrideConcurrency=true
 maxDistance=1850
 soundClassObject=&Fort_Loot_Collect
 volumeMultiplier=1
 concurrencyOverrides {maxCount=4, resolutionRule=PreventNew}
}
